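# M_Helios_BaseColor is a simple Default Lit test material in a Helios developer map folder, flagged for use with Nanite.
# A BaseColor vector parameter, pinkish red by default, feeds a Desaturation node that slightly mutes it by a constant fraction.
# A separate scalar parameter sits alongside, and the graph shows no links from either branch to the material outputs.
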Material /Game/Athena/Helios/Maps/Test/Dev/Stu/Materials/M_Helios_BaseColor
graph {
 properties {bUsedWithNanite=true, pixelDepthOffsetMode=PDOM_Legacy, usedShadingModels=Default Lit}
 nodes [
  Desaturation(None=VectorParameter() {parameterName=BaseColor, defaultValue=<rgba 1 0.2603 0.2703 1>}.RGB, Fraction=Constant() {r=0.1})
  ScalarParameter() {defaultValue=0.75}
 ]
}
